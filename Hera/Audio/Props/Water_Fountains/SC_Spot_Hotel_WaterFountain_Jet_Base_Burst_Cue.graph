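# Sound Cue for an outdoor hotel water fountain jet burst one-shot, layering a base burst with a spray burst.
# Each layer randomly picks one of twelve waves and varies its pitch, then a Mixer blends them with the spray slightly quieter.
# The Quality Level root connects only its first two slots to that mix, leaving the others empty.
# Overrides add low-pass distance attenuation, listener focus and capped concurrency, alongside a distance-scaled reverb send.
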SoundCue /Hera/Audio/Props/Water_Fountains/SC_Spot_Hotel_WaterFountain_Jet_Base_Burst_Cue
root=QualityLevel_0
nodes [
 {id=Mixer_0, children [Modulator_0,Modulator_1], properties {inputVolume [1,0.8]}}
 {id=Modulator_0, children [Random_0], properties {pitchMax=1.1, pitchMin=0.9}}
 {id=Modulator_1, children [Random_1], properties {pitchMax=1.1, pitchMin=1}}
 {id=QualityLevel_0, children [Mixer_0,Mixer_0,null,null]}
 {
  id=Random_0
  children [
   WavePlayer_0
   WavePlayer_1
   WavePlayer_2
   WavePlayer_3
   WavePlayer_4
   WavePlayer_5
   WavePlayer_6
   WavePlayer_7
   WavePlayer_8
   WavePlayer_9
   WavePlayer_10
   WavePlayer_11
  ]
  properties {weights [1,1,1,1,1,1,1,1,1,1,1,1]}
 }
 {
  id=Random_1
  children [
   WavePlayer_12
   WavePlayer_13
   WavePlayer_14
   WavePlayer_15
   WavePlayer_16
   WavePlayer_17
   WavePlayer_18
   WavePlayer_19
   WavePlayer_20
   WavePlayer_21
   WavePlayer_22
   WavePlayer_23
  ]
  properties {weights [1,1,1,1,1,1,1,1,1,1,1,1]}
 }
 {id=WavePlayer_0, properties {soundWaveAssetPtr=&/Spot_Hotel_Water_Fountain_Jet_Base_Burst_01}}
 {id=WavePlayer_1, properties {soundWaveAssetPtr=&/Spot_Hotel_Water_Fountain_Jet_Base_Burst_02}}
 {id=WavePlayer_10, properties {soundWaveAssetPtr=&/Spot_Hotel_Water_Fountain_Jet_Base_Burst_11}}
 {id=WavePlayer_11, properties {soundWaveAssetPtr=&/Spot_Hotel_Water_Fountain_Jet_Base_Burst_12}}
 {id=WavePlayer_12, properties {soundWaveAssetPtr=&/Spot_Hotel_Water_Fountain_Jet_Spray_Burst_01}}
 {id=WavePlayer_13, properties {soundWaveAssetPtr=&/Spot_Hotel_Water_Fountain_Jet_Spray_Burst_02}}
 {id=WavePlayer_14, properties {soundWaveAssetPtr=&/Spot_Hotel_Water_Fountain_Jet_Spray_Burst_03}}
 {id=WavePlayer_15, properties {soundWaveAssetPtr=&/Spot_Hotel_Water_Fountain_Jet_Spray_Burst_04}}
 {id=WavePlayer_16, properties {soundWaveAssetPtr=&/Spot_Hotel_Water_Fountain_Jet_Spray_Burst_05}}
 {id=WavePlayer_17, properties {soundWaveAssetPtr=&/Spot_Hotel_Water_Fountain_Jet_Spray_Burst_06}}
 {id=WavePlayer_18, properties {soundWaveAssetPtr=&/Spot_Hotel_Water_Fountain_Jet_Spray_Burst_07}}
 {id=WavePlayer_19, properties {soundWaveAssetPtr=&/Spot_Hotel_Water_Fountain_Jet_Spray_Burst_08}}
 {id=WavePlayer_2, properties {soundWaveAssetPtr=&/Spot_Hotel_Water_Fountain_Jet_Base_Burst_03}}
 {id=WavePlayer_20, properties {soundWaveAssetPtr=&/Spot_Hotel_Water_Fountain_Jet_Spray_Burst_09}}
 {id=WavePlayer_21, properties {soundWaveAssetPtr=&/Spot_Hotel_Water_Fountain_Jet_Spray_Burst_10}}
 {id=WavePlayer_22, properties {soundWaveAssetPtr=&/Spot_Hotel_Water_Fountain_Jet_Spray_Burst_11}}
 {id=WavePlayer_23, properties {soundWaveAssetPtr=&/Spot_Hotel_Water_Fountain_Jet_Spray_Burst_12}}
 {id=WavePlayer_3, properties {soundWaveAssetPtr=&/Spot_Hotel_Water_Fountain_Jet_Base_Burst_04}}
 {id=WavePlayer_4, properties {soundWaveAssetPtr=&/Spot_Hotel_Water_Fountain_Jet_Base_Burst_05}}
 {id=WavePlayer_5, properties {soundWaveAssetPtr=&/Spot_Hotel_Water_Fountain_Jet_Base_Burst_06}}
 {id=WavePlayer_6, properties {soundWaveAssetPtr=&/Spot_Hotel_Water_Fountain_Jet_Base_Burst_07}}
 {id=WavePlayer_7, properties {soundWaveAssetPtr=&/Spot_Hotel_Water_Fountain_Jet_Base_Burst_08}}
 {id=WavePlayer_8, properties {soundWaveAssetPtr=&/Spot_Hotel_Water_Fountain_Jet_Base_Burst_09}}
 {id=WavePlayer_9, properties {soundWaveAssetPtr=&/Spot_Hotel_Water_Fountain_Jet_Base_Burst_10}}
]
properties {
 bOverrideAttenuation=true
 bOverrideConcurrency=true
 maxDistance=4850
 soundClassObject=&Fort_Ambient_OneShots_Outdoors
 volumeMultiplier=1
 attenuationOverrides {
  bAttenuateWithLPF=true
  bEnableFocusInterpolation=true
  bEnableListenerFocus=true
  bEnableLogFrequencyScaling=true
  focusReleaseInterpSpeed=0.5
  lPFFrequencyAtMax=600
  lPFRadiusMax=3600
  lPFRadiusMin=2000
  nonFocusVolumeAttenuation=0.75
  attenuationShapeExtents {x=1250}
 }
 concurrencyOverrides {maxCount=4}
 soundSubmixSends [{
   maxSendDistance=3600
   maxSendLevel=0.6
   minSendDistance=1500
   minSendLevel=0.15
   sendLevelControlMethod=Linear
   soundSubmix=&Fortnite_Reverb_Default
  }]
}
