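# Sound cue for a looping water fountain jet spray ambience at a hotel, routed to the outdoor ambient one-shots sound class.
# All quality levels share one chain: a random pick from nine equally weighted looping waves, passed through a modulator that randomly raises pitch.
# Overrides give it capsule attenuation with a custom falloff, distance low-pass, HRTF spatialization, listener focus and a distance-scaled reverb send.
# A concurrency override caps simultaneous instances at four.
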SoundCue /Hera/Audio/Props/Water_Fountains/SC_Spot_Hotel_WaterFountain_Jet_Spray_Loop_Cue
root=QualityLevel_0
nodes [
 {id=Modulator_1, children [Random_1], properties {pitchMax=1.2, pitchMin=1}}
 {id=QualityLevel_0, children [Modulator_1,Modulator_1,Modulator_1,Modulator_1]}
 {
  id=Random_1
  children [WavePlayer_12, WavePlayer_13, WavePlayer_14, WavePlayer_15, WavePlayer_16, WavePlayer_17, WavePlayer_18, WavePlayer_19, WavePlayer_20]
  properties {weights [1,1,1,1,1,1,1,1,1]}
 }
 {
  id=WavePlayer_12
  properties {bLooping=true, soundWaveAssetPtr=&/Spot_Hotel_Water_Fountain_Jet_Spray_Loop_01}
 }
 {
  id=WavePlayer_13
  properties {bLooping=true, soundWaveAssetPtr=&/Spot_Hotel_Water_Fountain_Jet_Spray_Loop_02}
 }
 {
  id=WavePlayer_14
  properties {bLooping=true, soundWaveAssetPtr=&/Spot_Hotel_Water_Fountain_Jet_Spray_Loop_03}
 }
 {
  id=WavePlayer_15
  properties {bLooping=true, soundWaveAssetPtr=&/Spot_Hotel_Water_Fountain_Jet_Spray_Loop_04}
 }
 {
  id=WavePlayer_16
  properties {bLooping=true, soundWaveAssetPtr=&/Spot_Hotel_Water_Fountain_Jet_Spray_Loop_05}
 }
 {
  id=WavePlayer_17
  properties {bLooping=true, soundWaveAssetPtr=&/Spot_Hotel_Water_Fountain_Jet_Spray_Loop_06}
 }
 {
  id=WavePlayer_18
  properties {bLooping=true, soundWaveAssetPtr=&/Spot_Hotel_Water_Fountain_Jet_Spray_Loop_07}
 }
 {
  id=WavePlayer_19
  properties {bLooping=true, soundWaveAssetPtr=&/Spot_Hotel_Water_Fountain_Jet_Spray_Loop_08}
 }
 {
  id=WavePlayer_20
  properties {bLooping=true, soundWaveAssetPtr=&/Spot_Hotel_Water_Fountain_Jet_Spray_Loop_09}
 }
]
properties {
 bOverrideAttenuation=true
 bOverrideConcurrency=true
 maxDistance=4600
 soundClassObject=&Fort_Ambient_OneShots_Outdoors
 attenuationOverrides {
  attenuationShape=Capsule
  bAttenuateWithLPF=true
  bEnableFocusInterpolation=true
  bEnableListenerFocus=true
  bEnableLogFrequencyScaling=true
  distanceAlgorithm=Custom
  focusReleaseInterpSpeed=0.5
  lPFFrequencyAtMax=600
  lPFRadiusMax=3600
  lPFRadiusMin=2000
  nonFocusVolumeAttenuation=0.75
  spatializationAlgorithm=SPATIALIZATION_HRTF
  stereoSpread=0
  attenuationShapeExtents {x=1000}
  customAttenuationCurve {editorCurveData {keys [
     <at 0 1>
     <at 0.05 1>
     {arriveTangent=-1.0526, interpMode=RCIM_Cubic, leaveTangent=-1.0526, time=0.5, value=0.33}
     {arriveTangent=-1.3611, interpMode=RCIM_Cubic, leaveTangent=-1.3611, tangentMode=RCTM_User, time=1, value=0}
    ]}}
 }
 concurrencyOverrides {maxCount=4}
 soundSubmixSends [{
   maxSendDistance=3600
   maxSendLevel=0.6
   minSendDistance=1500
   minSendLevel=0.15
   sendLevelControlMethod=Linear
   soundSubmix=&Fortnite_Reverb_Default
  }]
}
